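# Niagara ribbon trail effect for the mint Snow Globe glider, drawing green and white smoke streaks from two bone/socket pairs on the glider.
# Main green and smoke emitters fade over particle life, with alpha also scaled by a system-level FN_Integrator value.
# Duplicate emitters spawn only when bIsFrontEnd is set, at a lower rate with constant -Y velocity to stream trails in the front end, bypassing the integrator fade.
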NiagaraSystem /BRCosmetics/Gadgets/Assets/VinderTech_GliderChute/Glider_Snow_Globe_Mint/VFX/NS_Glider_SnowGlobe
dependencies {
 usedDataInterfaces [&NiagaraDataInterfaceCurve]
 usedDynamicInputs []
 usedModules [&SystemState, &FN_Integrator]
}
emitters {
 FN_AnimTrail_Green {
  emitterUpdateScript [
   &EmitterState(Scalability Mode=::NewEnumerator1)
   /Game/Effects/Niagara/Modules/AnimTrail/FN_SpawnAnimTrail/FN_SpawnAnimTrail(Number of Anim Trails=2, SpawnRate=100)
  ]
  particleSpawnScript [
   /Game/Effects/Niagara/Modules/AnimTrail/FN_InitializeAnimTrail/FN_InitializeAnimTrailBoneIndecies(Second Bone or Socket Index=1)
   &FN_InitializeAnimTrailBoneIndecies(AnimTrail Selection=1, First Bone or Socket Index=2, Second Bone or Socket Index=3) {module=FN_InitializeAnimTrailBoneIndecies001}
   /Game/Effects/Niagara/Modules/AnimTrail/FN_InitializeAnimTrail/FN_InitializeAnimTrail(
    AnimTrailWidth=1
    Color=&MultiplyLinearColor(Linear Color=<rgba 0.2976 2 0 0.8>, Scale Factor=<1 1 1 1>)
    Lifetime=0.175
    Write Material Random=false
   )
  ]
  particleUpdateScript [
   &InheritVelocity(Module.Limit Speed=true, Velocity Limit=20)
   &SolveForcesAndVelocity(Clamp Velocity=false, Limit Acceleration=false)
   /Game/Effects/Niagara/Modules/AnimTrail/FN_UpdateAnimTrail/FN_UpdateAnimTrail(New float=1)
   &ScaleColor(
    Scale Alpha=&Multiply_Float(
     A=&FloatFromCurve(FloatCurve=DataInterface() {propertyValues {Curve {keys [
          {arriveTangent=3.6646, interpMode=RCIM_Cubic, leaveTangent=3.6646, tangentMode=RCTM_User, time=-0.0017, value=0.5086}
          {arriveTangent=-0.0143, interpMode=RCIM_Cubic, leaveTangent=-0.0143, tangentMode=RCTM_User, time=0.1233, value=0.7411}
          {arriveTangent=-1.317, interpMode=RCIM_Cubic, leaveTangent=-1.317, tangentMode=RCTM_User, time=1, value=0}
         ]}}})
     B=&FloatFromCurve(
      FloatCurve=DataInterface() {propertyValues {Curve {keys [<at 0 0 RCIM_Cubic>, {arriveTangent=2.0248, interpMode=RCIM_Cubic, leaveTangent=2.0248, tangentMode=RCTM_User, time=1, value=1}]}}}
      CurveIndex=$System.FN_Integrator.Value
     )
    )
    Scale RGB=&MakeVectorFromLinearColorRGB(Color=<rgba 1.4879 10 0 1>)
    ScaleA=true
    ScaleRGB=false
   )
  ]
  rendererClasses [&NiagaraRibbonRendererProperties]
  renderers [{rendererClass=&NiagaraRibbonRendererProperties}]
 }
 FN_AnimTrail_Green001 {
  emitterUpdateScript [
   &EmitterState(Scalability Mode=::NewEnumerator1)
   &FN_SpawnAnimTrail(
    Number of Anim Trails:NiagaraInt32=&MakeCustomIntFromBool(Bool=$User.bIsFrontEnd, True INT=2, False INT=0)
    SpawnRate=40
   )
  ]
  particleSpawnScript [
   &FN_InitializeAnimTrailBoneIndecies(Second Bone or Socket Index=1)
   &FN_InitializeAnimTrailBoneIndecies(AnimTrail Selection=1, First Bone or Socket Index=2, Second Bone or Socket Index=3) {module=FN_InitializeAnimTrailBoneIndecies001}
   &FN_InitializeAnimTrail(
    AnimTrailWidth=1
    Color=&MultiplyLinearColor(Linear Color=<rgba 0.2976 2 0 0.8>, Scale Factor=<1 1 1 1>)
    Lifetime=0.175
    Write Material Random=false
   )
  ]
  particleUpdateScript [
   &AddVelocity(Velocity=<0 -400 0>)
   &InheritVelocity(Module.Limit Speed=true, Velocity Limit=20)
   &SolveForcesAndVelocity(Clamp Velocity=false, Limit Acceleration=false)
   &FN_UpdateAnimTrail(New float=1)
   &ScaleColor(
    Scale Alpha=&Multiply_Float(
     A=&FloatFromCurve(FloatCurve=DataInterface() {propertyValues {Curve {keys [
          {arriveTangent=3.6646, interpMode=RCIM_Cubic, leaveTangent=3.6646, tangentMode=RCTM_User, time=-0.0017, value=0.5086}
          {arriveTangent=-0.0143, interpMode=RCIM_Cubic, leaveTangent=-0.0143, tangentMode=RCTM_User, time=0.1233, value=0.7411}
          {arriveTangent=-1.317, interpMode=RCIM_Cubic, leaveTangent=-1.317, tangentMode=RCTM_User, time=1, value=0}
         ]}}})
     B=1
    )
    Scale RGB=&MakeVectorFromLinearColorRGB(Color=<rgba 1.4879 10 0 1>)
    ScaleA=true
    ScaleRGB=false
   )
  ]
  rendererClasses [&NiagaraRibbonRendererProperties]
  renderers [{rendererClass=&NiagaraRibbonRendererProperties}]
 }
 FN_AnimTrail_Smoke {
  emitterUpdateScript [
   &EmitterState(Scalability Mode=::NewEnumerator1)
   &FN_SpawnAnimTrail(Number of Anim Trails=2, SpawnRate=100)
  ]
  particleSpawnScript [
   &FN_InitializeAnimTrailBoneIndecies(Second Bone or Socket Index=1)
   &FN_InitializeAnimTrailBoneIndecies(AnimTrail Selection=1, First Bone or Socket Index=2, Second Bone or Socket Index=3) {module=FN_InitializeAnimTrailBoneIndecies001}
   &FN_InitializeAnimTrail(
    AnimTrailWidth=1
    Color=&MultiplyLinearColor(Linear Color=<rgba 1 1 1 1.5>, Scale Factor=<1 1 1 1>)
    Lifetime=0.12
    Write Material Random=false
   )
   &CameraOffset(Camera Offset Amount=&RandomRangeFloat(Minimum=5, Maximum=15))
  ]
  particleUpdateScript [
   &InheritVelocity(Module.Limit Speed=true, Velocity Limit=20)
   &SolveForcesAndVelocity(Clamp Velocity=false, Limit Acceleration=false)
   &FN_UpdateAnimTrail(New float=1)
   &ScaleColor(
    Scale Alpha=&Multiply_Float(
     A=&FloatFromCurve(FloatCurve=DataInterface() {propertyValues {Curve {keys [
          {arriveTangent=3.6646, interpMode=RCIM_Cubic, leaveTangent=3.6646, tangentMode=RCTM_User, time=-0.0017, value=0.5086}
          {arriveTangent=-0.0143, interpMode=RCIM_Cubic, leaveTangent=-0.0143, tangentMode=RCTM_User, time=0.1233, value=0.7411}
          {arriveTangent=-1.317, interpMode=RCIM_Cubic, leaveTangent=-1.317, tangentMode=RCTM_User, time=1, value=0}
         ]}}})
     B=&FloatFromCurve(
      FloatCurve=DataInterface() {propertyValues {Curve {keys [<at 0 0 RCIM_Cubic>, {arriveTangent=2.0248, interpMode=RCIM_Cubic, leaveTangent=2.0248, tangentMode=RCTM_User, time=1, value=1}]}}}
      CurveIndex=$System.FN_Integrator.Value
     )
    )
    Scale RGB=&MakeVectorFromLinearColorRGB(Color=<rgba 1.4879 10 0 1>)
    ScaleA=true
    ScaleRGB=false
   )
  ]
  rendererClasses [&NiagaraRibbonRendererProperties]
  renderers [{rendererClass=&NiagaraRibbonRendererProperties}]
 }
 FN_AnimTrail_Smoke001 {
  emitterUpdateScript [
   &EmitterState(Scalability Mode=::NewEnumerator1)
   &FN_SpawnAnimTrail(
    Number of Anim Trails:NiagaraInt32=&MakeCustomIntFromBool(Bool=$User.bIsFrontEnd, True INT=2, False INT=0)
    SpawnRate=40
   )
  ]
  particleSpawnScript [
   &FN_InitializeAnimTrailBoneIndecies(Second Bone or Socket Index=1)
   &FN_InitializeAnimTrailBoneIndecies(AnimTrail Selection=1, First Bone or Socket Index=2, Second Bone or Socket Index=3) {module=FN_InitializeAnimTrailBoneIndecies001}
   &FN_InitializeAnimTrail(
    AnimTrailWidth=1
    Color=&MultiplyLinearColor(Linear Color=<rgba 1 1 1 1.5>, Scale Factor=<1 1 1 1>)
    Lifetime=0.12
    Write Material Random=false
   )
   &CameraOffset(Camera Offset Amount=&RandomRangeFloat(Minimum=5, Maximum=15)) {enabled=false}
  ]
  particleUpdateScript [
   &AddVelocity(Velocity=<0 -500 0>) {module=AddVelocity001}
   &InheritVelocity(Module.Limit Speed=true, Velocity Limit=20)
   &SolveForcesAndVelocity(Clamp Velocity=false, Limit Acceleration=false)
   &FN_UpdateAnimTrail(New float=1)
   &ScaleColor(
    Scale Alpha=&Multiply_Float(
     A=&FloatFromCurve(FloatCurve=DataInterface() {propertyValues {Curve {keys [
          {arriveTangent=3.6646, interpMode=RCIM_Cubic, leaveTangent=3.6646, tangentMode=RCTM_User, time=-0.0017, value=0.5086}
          {arriveTangent=-0.0143, interpMode=RCIM_Cubic, leaveTangent=-0.0143, tangentMode=RCTM_User, time=0.1233, value=0.7411}
          {arriveTangent=-1.317, interpMode=RCIM_Cubic, leaveTangent=-1.317, tangentMode=RCTM_User, time=1, value=0}
         ]}}})
     B=1
    )
    Scale RGB=&MakeVectorFromLinearColorRGB(Color=<rgba 1.4879 10 0 1>)
    ScaleA=true
    ScaleRGB=false
   )
  ]
  rendererClasses [&NiagaraRibbonRendererProperties]
  renderers [{rendererClass=&NiagaraRibbonRendererProperties}]
 }
}
summary {
 systemName=NS_Glider_SnowGlobe
 emitters [
  {emitterName=FN_AnimTrail_Green, rendererClasses [&NiagaraRibbonRendererProperties]}
  {emitterName=FN_AnimTrail_Smoke, rendererClasses [&NiagaraRibbonRendererProperties]}
  {emitterName=FN_AnimTrail_Green001, rendererClasses [&NiagaraRibbonRendererProperties]}
  {emitterName=FN_AnimTrail_Smoke001, rendererClasses [&NiagaraRibbonRendererProperties]}
 ]
 userVariables [
  User.bIsFrontEnd {type=NiagaraBool}
  User.ForwardDot {type=NiagaraFloat, defaultValue=0, description=""}
  User.IsFullyDeployed {type=NiagaraBool}
  User.RibbonIDOffset {type=NiagaraInt32}
  User.RightDot {type=NiagaraFloat, defaultValue=0, description=""}
  User.SkeletalMesh {
   type=NiagaraDataInterfaceSkeletalMesh
   defaultValue {
    struct=NiagaraExt_VariableValue_DataInterface
    value {dataInterface=&self:NiagaraDataInterfaceSkeletalMesh_0}
   }
  }
 ]
}
